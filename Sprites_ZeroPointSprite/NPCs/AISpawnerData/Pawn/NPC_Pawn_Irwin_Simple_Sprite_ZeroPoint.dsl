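; Pawn Blueprint for the Zero Point sprite NPC, which applies its own materials and attaches a cape component to the spine, hidden on death.
; UpdateAlerted picks a non-crouching player from the awareness list as leader, mirrored to the blackboard.
; Detected-player changes clear a leader beyond a release range, using a cached squared distance.
; Construction writes the data-driven approach distance to the blackboard.
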(event EventBeginPlay
  (bind _mesh (Variables|Character|GetMesh))
  (|CallParentFunction)
  (Rendering|Material|SetMaterial _mesh 0 "/Sprites_ZeroPointSprite/Assets/Materials/MI_ZeroPoint_Sprite.MI_ZeroPoint_Sprite")
  (Rendering|Material|SetMaterial _mesh 1 "/Sprites_ZeroPointSprite/Assets/Materials/M_Sprite_ZeroPoint_Facial.M_Sprite_ZeroPoint_Facial")
  (Rendering|Material|SetMaterial _mesh 2 "/Game/Global/Materials/M_Overlay_VertexCrunch.M_Overlay_VertexCrunch"))

(event EventOnDetectedPlayerChangedEvent
  (params (AllFoundActors :type Array%20of%20Actor%20Object%20References) (EnteringActors :type Array%20of%20Actor%20Object%20References) (ExitingActors :type Array%20of%20Actor%20Object%20References))
  (bind _leader (Variables|Default|GetLeader))
  (if (not (> (Variables|Default|GetCachedReleaseLeaderSquaredRange) 0))
    (Variables|Default|SetCachedReleaseLeaderSquaredRange (Math|Float|Square (ScalableFloat|GetValueatLevel (Variables|Default|GetRow_ReleaseLeaderRange)))))
  ; If we're too far from our current leader, release it and allow the sprite to find a new one
  (Utilities|IsValid _leader
    (:Is%20Valid
      (if (>= (Transformation|GetSquaredDistanceTo _leader self) (Variables|Default|GetCachedReleaseLeaderSquaredRange))
        (CallFunction|SetLeader 0)))
    (:Is%20Not%20Valid
      _))
  (|CallParentFunction AllFoundActors EnteringActors ExitingActors))

(event EventOnDeathPlayEffects
  (params (Damage :type Float%20%28single-precision%29) (DamageTags :type Gameplay%20Tag%20Container%20Structure%20%28by%20ref%29) (Momentum :type Vector :default "0, 0, 0") (HitInfo :type Hit%20Result%20Structure%20%28by%20ref%29) (InstigatedBy :type Fort%20Pawn%20Object%20Reference) (DamageCauser :type Actor%20Object%20Reference) (EffectContext :type Gameplay%20Effect%20Context%20Handle%20Structure))
  (Rendering|SetVisibility (Variables|Default|GetSprite_Cape_Creature_Scale) false true)
  (|CallParentFunction Damage DamageTags Momentum HitInfo InstigatedBy DamageCauser EffectContext))

(fn GetSpriteType
  (return "NewEnumerator4"))

(fn SetLeader
  (params (NewLeader :type Actor%20Object%20Reference))
  (AI|Components|Blackboard|SetValueasObject (AI|GetBlackboard self) (Variables|Default|GetLeaderKey) NewLeader)
  (Variables|Default|SetLeader NewLeader))

(fn UpdateAlerted
  (bind _playerpawnarray (Variables|Awareness|GetPlayerPawnArray))
  (bind _foundleader (LocalVariables|GetFoundLeader))
  ; For all players this sprite is aware of, find the first that isn't crouching and assign it as the sprite's leader
  (LocalVariables|Setloc_bIsAlerted false)
  (if (not (Utilities|Array|IsEmpty _playerpawnarray))
    (bind _array_element (Utilities|Array|ForEachLoopwithBreak _playerpawnarray)
      (:LoopBody
        (Utilities|IsValid _array_element
          (:Is%20Valid
            (if (not (GameplayTags|HasMatchingGameplayTag _array_element "(TagName=\"Gameplay.Action.Player.Crouch\")"))
              (LocalVariables|SetFoundLeader _array_element)
              (LocalVariables|Setloc_bIsAlerted true)))))
      (:Completed
        _))
    (Utilities|IsValid _foundleader
      (:Is%20Valid
        (CallFunction|SetLeader _foundleader)))))

(fn ConstructionScript
  (|CallParentFunction)
  (AI|Components|Blackboard|SetValueasFloat (AI|GetBlackboard 0) (Variables|Default|GetKey_ApproachDistance) (ScalableFloat|GetValueatLevel (Variables|Default|GetRow_ApproachIdleRange)))
  (Transformation|AttachComponentToComponent (Variables|Default|GetSprite_Cape_Creature_Scale) (Variables|Character|GetMesh) "spine_1" "KeepRelative" "KeepRelative" "KeepRelative" true))
